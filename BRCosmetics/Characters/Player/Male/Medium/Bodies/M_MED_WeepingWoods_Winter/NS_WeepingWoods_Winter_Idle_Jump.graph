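# Niagara cosmetic effect for the WeepingWoods Winter player body, shedding leaves and snow at idle and on landing.
# Idle emitters trickle mesh leaves and snowflake sprites off the skeletal mesh, with a front-end modifier zeroing their spawn rate.
# User.LandBurst gates dense landing emitters: leaves, flakes, and head and torso snow puffs, all inheriting velocity.
# Puffs use SubUV animation and distance-based alpha from the owner; head puffs spawn in an offset sphere.
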NiagaraSystem /BRCosmetics/Characters/Player/Male/Medium/Bodies/M_MED_WeepingWoods_Winter/NS_WeepingWoods_Winter_Idle_Jump
dependencies {
 usedDataInterfaces [&NiagaraDataInterfaceColorCurve, &NiagaraDataInterfaceCurve]
 usedDynamicInputs [
  &Boolean_AndOperation
  &SetBoolByFloatComparison
  &MakeCustomFloatFromBool
 ]
 usedModules [
  &SystemState
  &FN_Integrator
  &self:SystemSpawnScript.SystemScriptSource.SystemScriptGraph.NiagaraNodeAssignment_0.SetVariables_#1
  &self:SystemSpawnScript.SystemScriptSource.SystemScriptGraph.NiagaraNodeAssignment_1.SetVariables_#2
 ]
}
emitters {
 Idle_Leaves {
  emitterSpawnScript [/SetVariables(Emitter.RandomSeed:NiagaraInt32=/Niagara/DynamicInputs/Random/RandomInt(Randomness Mode:ENiagaraRandomnessMode=::NewEnumerator0, Random Seed:NiagaraInt32=0, Fixed Random Seed:NiagaraBool=false)) {isSetParameters=true, module=SetVariables_#3}]
  emitterUpdateScript [
   &EmitterState(Life Cycle Mode=::NewEnumerator1, Loop Duration=1, UseLoopDelay=false)
   &SpawnRate(
    SpawnRate=&FN_FrontEndFloatModifier(Gameplay=&RandomRangeFloat(Minimum=1, Maximum=2), Front End=0)
    Use Spawn Probability=true
   )
  ]
  particleSpawnScript [
   /Niagara/Modules/Spawn/Initialization/InitializeParticle(
    Color=&ColorFromCurve(
     ColorCurve=DataInterface() {propertyValues {
       AlphaCurve {keys [<at 0 1 RCIM_Cubic>,<at 1 1 RCIM_Cubic>]}
       BlueCurve {keys [<at 0.0624 0.1718 RCIM_Cubic>,<at 0.3159 1>,<at 1 1 RCIM_Cubic>]}
       GreenCurve {keys [<at 0.0624 0.7225 RCIM_Cubic>,<at 0.3159 0.8878>,<at 1 1 RCIM_Cubic>]}
       RedCurve {keys [<at 0.0624 1 RCIM_Cubic>,<at 0.3159 0.6963>,<at 1 1 RCIM_Cubic>]}
      }}
     CurveIndex=&RandomRangeFloat(Minimum=0, Maximum=1)
    )
    Lifetime=&RandomRangeFloat(Minimum=0.8, Maximum=1.5)
    Mesh Scale=&Multiply_VectorByFloat(
     Vector=<1 1 1>
     Float=&FloatFromCurve(
      FloatCurve=DataInterface() {propertyValues {Curve {keys [
          {arriveTangent=1.24, interpMode=RCIM_Cubic, leaveTangent=1.24, tangentMode=RCTM_User, time=0, value=0.7}
          {arriveTangent=0.0925, interpMode=RCIM_Cubic, leaveTangent=0.0925, tangentMode=RCTM_User, time=0.7, value=1}
          <at 1 1.2>
         ]}}}
      CurveIndex=&RandomRangeFloat(Minimum=0, Maximum=1)
     )
    )
    Module.Write Color=true
    Module.Write Lifetime=true
    Module.Write Mass=true
    Module.Write Material Random=true
    Module.Write Position=false
    Module.Write Scale=true
    Module.Write SpriteRotation=true
    Module.Write SpriteSize=true
    Position=&SimulationPosition()
    Sprite Rotation=&RandomRangeFloat(Minimum=0, Maximum=360)
    Sprite Size=&Vector2DFromFloat(Value=&RandomRangeFloat(Minimum=8, Maximum=10))
   )
   &SkeletalMeshLocation(Inherit Velocity=false, MaskSpawnGroup=false, Module.Override Randomness=false, Skeletal Mesh=$User.SkeletalMesh)
   &AddVelocity(Velocity=<0 0 -50>) {module=AddVelocity001}
   /Niagara/Modules/Update/Mass/CalculateSizeAndRotationalInertiaByMass(Density by Material Type:ENiagaraMassByVolume=::NewEnumerator4, Depth=0, Height=5, Module.Manually Enter Density=false)
   &InitialMeshOrientation(
    Orientation Vector=&Multiply_VectorByFloat(Vector=$Particles.Velocity, Float=-1)
    Rotation=&RandomRangeVector(Minimum=<0 0 0>, Maximum=<1 0 0>)
    Use Orientation Vector=true
    Use Rotation Vector=true
   )
   &ApplyInitialForces()
   /SetVariables(Particles.RandRotForce:NiagaraFloat=&RandomRangeFloat(Minimum=-1.5, Maximum=1.5)) {isSetParameters=true, module=SetVariables_#4}
   &AddVelocityFromPoint(Module.Use Falloff Distance=false, Origin Offset=<0 0 110>, Velocity Strength=30) {module=AddVelocityFromPoint001}
  ]
  particleUpdateScript [
   &AccelerationForce(Acceleration=&RandomRangeVector(Minimum=<0 0 -60>, Maximum=<0 0 -100>))
   &ScaleColor(Scale RGB=&RandomRangeVector(Minimum=<4 4 4>, Maximum=<6 6 6>), ScaleA=true, ScaleRGB=true)
   &Drag(Use Linear Drag=true, Use Rotational Drag=false)
   &CurlNoiseForce(Noise Strength=100, Pan Noise=false)
   &SolveForcesAndVelocity(Clamp Velocity=false, Limit Acceleration=false)
   &SolveRotationalForcesAndVelocity()
   &ParticleState()
   &UpdateMeshOrientation(Rotation Vector=&Lerp_Vector(
     A=<0 0 0>
     B=&Multiply_VectorByFloat(Vector=<1 1 1>, Float=$Particles.RandRotForce)
     Alpha=&FloatFromCurve(FloatCurve=DataInterface() {propertyValues {Curve {keys [
          <at 0 0>
          {arriveTangent=-0.0302, interpMode=RCIM_Cubic, leaveTangent=-0.0302, tangentMode=RCTM_User, time=0.2, value=1.5}
          {arriveTangent=-2.2278, interpMode=RCIM_Cubic, leaveTangent=-2.2278, tangentMode=RCTM_User, time=1, value=0.1}
         ]}}})
    ))
   &ScaleMeshSize(Scale Factor=&Multiply_VectorByFloat(
     Vector=<1 1 1>
     Float=&FloatFromCurve(FloatCurve=DataInterface() {propertyValues {Curve {keys [
          <at 0 1>
          {arriveTangent=-0.0008, interpMode=RCIM_Cubic, leaveTangent=-0.0008, tangentMode=RCTM_User, time=0.4, value=1}
          {arriveTangent=-3.4597, interpMode=RCIM_Cubic, leaveTangent=-3.4598, tangentMode=RCTM_User, time=1, value=0.2}
         ]}}})
    ))
  ]
  rendererClasses [&NiagaraMeshRendererProperties]
  renderers [{rendererClass=&NiagaraMeshRendererProperties, rendererIndex=0}]
 }
 Idle_SnowFlakes {
  emitterUpdateScript [
   &EmitterState(Life Cycle Mode=::NewEnumerator1, Loop Duration=1, UseLoopDelay=false)
   &SpawnRate(
    SpawnRate=&FN_FrontEndFloatModifier(Gameplay=&RandomRangeFloat(Minimum=2, Maximum=4), Front End=0)
    Use Spawn Probability=true
   )
  ]
  particleSpawnScript [
   &InitializeParticle(
    Lifetime=&RandomRangeFloat(Minimum=0.7, Maximum=1.2)
    Module.Write Color=true
    Module.Write Lifetime=true
    Module.Write Mass=true
    Module.Write Material Random=true
    Module.Write Position=false
    Module.Write Scale=false
    Module.Write SpriteRotation=true
    Module.Write SpriteSize=true
    Position=&SimulationPosition()
    Sprite Rotation=&RandomRangeFloat(Minimum=0, Maximum=360)
    Sprite Size=&Vector2DFromFloat(Value=&RandomRangeFloat(Minimum=3, Maximum=5))
   )
   &Color(Color=<rgba 0.4 0.59 1 1>, Scale Alpha=$Particles.MaterialRandom)
   &TorusLocation(
    Align Torus With Axis=false
    Handle Radius=0
    Large Radius:NiagaraFloat=&RandomRangeFloat(Minimum=15, Maximum=18)
    MaskSpawnGroup=false
    Module.Override Randomness=false
   ) {enabled=false}
   &SkeletalMeshLocation(
    Bone Sampling Mode=::NewEnumerator0
    Inherit Velocity=false
    MaskSpawnGroup=false
    Module.Override Randomness=false
    Skeletal Mesh=$User.SkeletalMesh
   )
   &AddVelocityFromPoint(
    Module.Use Falloff Distance=false
    Origin Offset=<0 0 100>
    Velocity Strength=&RandomRangeFloat(Minimum=50, Maximum=75)
   )
   &AddVelocity(Velocity=<0 0 -10>)
   /SetVariables(Particles.RandFacing:Vector3f=&RandomRangeVector(Minimum=<-1 -1 -1>, Maximum=<1 1 1>)) {isSetParameters=true, module=SetVariables_#5}
  ]
  particleUpdateScript [
   &ParticleState()
   &CurlNoiseForce(Noise Frequency=100, Noise Strength=200, Pan Noise=false)
   &AccelerationForce(Acceleration=<0 0 -50>)
   &Drag(Drag=&RandomRangeFloat(Minimum=2, Maximum=3), Use Linear Drag=true, Use Rotational Drag=false)
   &SolveForcesAndVelocity(Clamp Velocity=false, Limit Acceleration=false)
   &ScaleColor(
    Scale RGB=&MakeVectorFromLinearColorRGB(Color=&ColorFromCurve(ColorCurve=DataInterface() {propertyValues {
        AlphaCurve {keys [<at 0 0 RCIM_Cubic>,<at 0.04 1>,<at 0.135 0.62>,<at 1.0115 0>,<at 1.094 0>]}
        BlueCurve {keys [<at 0 1 RCIM_Cubic>,<at 1 1 RCIM_Cubic>]}
        GreenCurve {keys [<at 0 0.8875 RCIM_Cubic>,<at 1 0.79 RCIM_Cubic>]}
        RedCurve {keys [<at 0 0.85 RCIM_Cubic>,<at 1 0.7 RCIM_Cubic>]}
       }}))
    ScaleA=true
    ScaleRGB=true
   )
   &ScaleSpriteSize(Scale Factor=&Vector2DFromFloat(Value=&FloatFromCurve(FloatCurve=DataInterface() {propertyValues {Curve {keys [
          <at 0 0 RCIM_Cubic>
          {arriveTangent=2, interpMode=RCIM_Cubic, leaveTangent=2, time=0.1619, value=1}
          {arriveTangent=-1.1932, interpMode=RCIM_Cubic, leaveTangent=-1.1932, time=0.5, value=1}
          <at 1 0 RCIM_Cubic>
         ]}}})))
   /NiagaraFN/Sprite/FN_SpriteAlignmentAndFacing(SpriteFacing:Vector3f=&Lerp_Vector(
     A=&RandomRangeVector(Minimum=<-1 -1 -1>, Maximum=<1 1 1>)
     B=$Particles.RandFacing
     Alpha=$Particles.NormalizedAge
    ))
  ]
  rendererClasses [&NiagaraSpriteRendererProperties]
  renderers [{rendererClass=&NiagaraSpriteRendererProperties}]
 }
 Land_Leaves {
  emitterSpawnScript [/SetVariables(Emitter.RandomSeed:NiagaraInt32=&RandomInt(Randomness Mode:ENiagaraRandomnessMode=::NewEnumerator0, Random Seed:NiagaraInt32=0, Fixed Random Seed:NiagaraBool=false)) {isSetParameters=true, module=SetVariables_#3}]
  emitterUpdateScript [
   &EmitterState(Life Cycle Mode=::NewEnumerator1, Loop Duration=1, UseLoopDelay=false)
   &SpawnRate(Spawn Probability=$User.LandBurst, SpawnRate=150, Use Spawn Probability=true)
  ]
  particleSpawnScript [
   &InitializeParticle(
    Color=&ColorFromCurve(
     ColorCurve=DataInterface() {propertyValues {
       AlphaCurve {keys [<at 0 1 RCIM_Cubic>,<at 1 1 RCIM_Cubic>]}
       BlueCurve {keys [<at 0.0624 0.5312 RCIM_Cubic>,<at 0.3159 1>,<at 1 1 RCIM_Cubic>]}
       GreenCurve {keys [<at 0.0624 0.843 RCIM_Cubic>,<at 0.3159 0.8878>,<at 1 1 RCIM_Cubic>]}
       RedCurve {keys [<at 0.0624 1 RCIM_Cubic>,<at 0.3159 0.6963>,<at 1 1 RCIM_Cubic>]}
      }}
     CurveIndex=&RandomRangeFloat(Minimum=0, Maximum=1)
    )
    Lifetime=&RandomRangeFloat(Minimum=0.8, Maximum=1.5)
    Mesh Scale=&Multiply_VectorByFloat(
     Vector=<1 1 1>
     Float=&FloatFromCurve(
      FloatCurve=DataInterface() {propertyValues {Curve {keys [
          {arriveTangent=1.24, interpMode=RCIM_Cubic, leaveTangent=1.24, tangentMode=RCTM_User, time=0, value=0.7}
          {arriveTangent=0.0925, interpMode=RCIM_Cubic, leaveTangent=0.0925, tangentMode=RCTM_User, time=0.7, value=1}
          <at 1 1.2>
         ]}}}
      CurveIndex=&RandomRangeFloat(Minimum=0, Maximum=1)
     )
    )
    Module.Write Color=true
    Module.Write Lifetime=true
    Module.Write Mass=true
    Module.Write Material Random=true
    Module.Write Position=false
    Module.Write Scale=true
    Module.Write SpriteRotation=true
    Module.Write SpriteSize=true
    Position=&SimulationPosition()
    Sprite Rotation=&RandomRangeFloat(Minimum=0, Maximum=360)
    Sprite Size=&Vector2DFromFloat(Value=&RandomRangeFloat(Minimum=8, Maximum=10))
   )
   &SkeletalMeshLocation(Inherit Velocity=false, MaskSpawnGroup=false, Module.Override Randomness=false, Skeletal Mesh=$User.SkeletalMesh)
   &AddVelocity(Velocity=<0 0 -50>) {module=AddVelocity001}
   &CalculateSizeAndRotationalInertiaByMass(Density by Material Type:ENiagaraMassByVolume=::NewEnumerator4, Depth=0, Height=5, Module.Manually Enter Density=false)
   &InitialMeshOrientation(
    Orientation Vector=&Multiply_VectorByFloat(Vector=$Particles.Velocity, Float=-1)
    Rotation=&RandomRangeVector(Minimum=<0 0 0>, Maximum=<1 0 0>)
    Use Orientation Vector=true
    Use Rotation Vector=true
   )
   &ApplyInitialForces()
   /SetVariables(Particles.RandRotForce:NiagaraFloat=&RandomRangeFloat(Minimum=-2, Maximum=2)) {isSetParameters=true, module=SetVariables_#4}
   &AddVelocityFromPoint(Module.Use Falloff Distance=false, Origin Offset=<0 0 100>, Velocity Strength=30) {module=AddVelocityFromPoint001}
   &InheritVelocity(Module.Limit Speed=true, Velocity Scale=<0.3 0.3 0.3>)
  ]
  particleUpdateScript [
   &AccelerationForce(Acceleration=&RandomRangeVector(Minimum=<0 0 -50>, Maximum=<0 0 -100>))
   &ScaleColor(Scale RGB=&RandomRangeVector(Minimum=<4 4 4>, Maximum=<6 6 6>), ScaleA=true, ScaleRGB=true)
   &Drag(Use Linear Drag=true, Use Rotational Drag=false)
   &CurlNoiseForce(Noise Frequency=100, Noise Strength=200, Pan Noise=false)
   &SolveForcesAndVelocity(Clamp Velocity=false, Limit Acceleration=false)
   &SolveRotationalForcesAndVelocity()
   &ParticleState()
   &UpdateMeshOrientation(Rotation Vector=&Lerp_Vector(
     A=<0 0 0>
     B=&Multiply_VectorByFloat(Vector=<1 1 1>, Float=$Particles.RandRotForce)
     Alpha=&FloatFromCurve(FloatCurve=DataInterface() {propertyValues {Curve {keys [
          <at 0 0>
          {arriveTangent=-0.0302, interpMode=RCIM_Cubic, leaveTangent=-0.0302, tangentMode=RCTM_User, time=0.2, value=1.5}
          {arriveTangent=-2.2278, interpMode=RCIM_Cubic, leaveTangent=-2.2278, tangentMode=RCTM_User, time=1, value=0.1}
         ]}}})
    ))
   &ScaleMeshSize(Scale Factor=&Multiply_VectorByFloat(
     Vector=<1 1 1>
     Float=&FloatFromCurve(FloatCurve=DataInterface() {propertyValues {Curve {keys [
          <at 0 1>
          {arriveTangent=-0.0008, interpMode=RCIM_Cubic, leaveTangent=-0.0008, tangentMode=RCTM_User, time=0.4, value=1}
          {arriveTangent=-3.4597, interpMode=RCIM_Cubic, leaveTangent=-3.4598, tangentMode=RCTM_User, time=1, value=0.2}
         ]}}})
    ))
   /Game/Effects/Niagara/Emitters/Water/FN_DistanceKillParticles(AgeReduiceAmount=0.01, DistanceFalloffExp=0.1) {enabled=false}
  ]
  rendererClasses [&NiagaraMeshRendererProperties]
  renderers [{rendererClass=&NiagaraMeshRendererProperties, rendererIndex=0}]
 }
 Land_SnowPuffs_Head {
  emitterSpawnScript [/SetVariables(Emitter.RandomSeed:NiagaraInt32=&RandomInt(Randomness Mode:ENiagaraRandomnessMode=::NewEnumerator0, Random Seed:NiagaraInt32=0, Fixed Random Seed:NiagaraBool=false)) {isSetParameters=true, module=SetVariables_#3}]
  emitterUpdateScript [
   &EmitterState(Life Cycle Mode=::NewEnumerator1, Loop Delay=0.2292, Loop Duration=1, UseLoopDelay=false)
   &SpawnRate(Spawn Probability=$User.LandBurst, SpawnRate=160, Use Spawn Probability=true)
  ]
  particleSpawnScript [
   &InitializeParticle(
    Color=&ColorFromCurve(
     ColorCurve=DataInterface() {propertyValues {
       AlphaCurve {keys [<at 0 0.06 RCIM_Cubic>,<at 1 0.02 RCIM_Cubic>]}
       BlueCurve {keys [<at 0 1 RCIM_Cubic>,<at 1 1 RCIM_Cubic>]}
       GreenCurve {keys [<at 0 1 RCIM_Cubic>,<at 1 0.8149 RCIM_Cubic>]}
       RedCurve {keys [<at 0 1 RCIM_Cubic>,<at 1 0.8 RCIM_Cubic>]}
      }}
     CurveIndex=$Particles.Partition.NormalizedID
    )
    Lifetime=&RandomRangeFloat(Minimum=0.7, Maximum=1.2)
    Module.Write Color=true
    Module.Write Lifetime=true
    Module.Write Mass=true
    Module.Write Material Random=true
    Module.Write Position=false
    Module.Write Scale=false
    Module.Write SpriteRotation=true
    Module.Write SpriteSize=true
    Position=&SimulationPosition()
    Sprite Rotation=&RandomRangeFloat(Minimum=-180, Maximum=180)
    Sprite Size=&Vector2DFromFloat(Value=&RandomRangeFloat(Minimum=6, Maximum=8))
   )
   &SkeletalMeshLocation(Bone Sampling Mode=::NewEnumerator4, Inherit Velocity=false, MaskSpawnGroup=false, Skeletal Mesh=$User.SkeletalMesh) {enabled=false}
   &SphereLocation(
    MaskSpawnGroup=false
    Offset=<0 0 90>
    Override Seed=false
    Sphere Radius=20
    Surface Only=false
    UseNonUniformScale=false
    UseOffset=true
    UseSphereOrientation=false
   )
   &AddVelocity(Velocity=&Multiply_VectorByFloat(Vector=$Output.FN_BoneFlow.SplineTangent, Float=10))
   &AddVelocity(Velocity=<0 0 20>) {module=AddVelocity001}
   &InheritVelocity(Module.Limit Speed=true, Velocity Scale=<0.5 0.5 0.5>)
  ]
  particleUpdateScript [
   &AccelerationForce(Acceleration=&RandomRangeVector(Minimum=<0 0 -60>, Maximum=<0 0 -100>))
   &AddVelocityFromPoint(
    Module.Use Falloff Distance=false
    Velocity Strength=&FloatFromCurve(
     FloatCurve=DataInterface() {propertyValues {Curve {keys [
         {interpMode=RCIM_Cubic, tangentMode=RCTM_User, time=0, value=0}
         {arriveTangent=0.0249, interpMode=RCIM_Cubic, leaveTangent=-3.8646, tangentMode=RCTM_Break, time=0.3, value=1}
         <at 1 0>
        ]}}}
     Scale Curve=7
    )
   )
   &Drag(Drag=2, Use Linear Drag=true, Use Rotational Drag=false)
   &SolveForcesAndVelocity(Clamp Velocity=false, Limit Acceleration=false)
   &ParticleState()
   &SubUVAnimation(
    Sub UV Play Rate:NiagaraFloat=&Multiply_Float(A=0.5, B=$Particles.NormalizedAge)
    SubUV Animation Mode:ENiagara_SubUVLookupMode=::NewEnumerator3
   ) {module=SubUVAnimation001}
   &SpriteRotationRate(Rotation Rate=&FloatFromCurve(
     FloatCurve=DataInterface() {propertyValues {Curve {keys [<at 0 0 RCIM_Cubic>,<at 0.3 1 RCIM_Cubic>,<at 1 0>]}}}
     Scale Curve=&RandomRangeFloat(Minimum=-20, Maximum=20)
    ))
   &ScaleSpriteSize(Scale Factor=&Vector2DFromFloat(Value=&FloatFromCurve(
      FloatCurve=DataInterface() {propertyValues {Curve {keys [
          {arriveTangent=0.7564, interpMode=RCIM_Cubic, leaveTangent=0.7564, tangentMode=RCTM_User, time=-0.0024, value=0.2}
          {arriveTangent=1.4669, interpMode=RCIM_Cubic, leaveTangent=1.4669, time=0.3797, value=1.3341}
          <at 0.9951 1.6633 RCIM_Cubic>
         ]}}}
      Scale Curve=3
     )))
   &ScaleColor(
    Scale Mode=::NewEnumerator1
    Scale RGBA=&MakeVector4_FromLinearColor(LinearColor=&ColorFromCurve(ColorCurve=DataInterface() {propertyValues {
        AlphaCurve {keys [
          <at 0 0 RCIM_Cubic>
          {arriveTangent=0.6593, interpMode=RCIM_Cubic, leaveTangent=0.6593, time=0.2192, value=0.5}
          {arriveTangent=-1.2318, interpMode=RCIM_Cubic, leaveTangent=-1.2318, tangentMode=RCTM_User, time=0.5309, value=0.35}
          <at 1.0115 0>
          <at 1.094 0>
         ]}
        BlueCurve {keys [<at 0 1 RCIM_Cubic>,<at 1 1 RCIM_Cubic>]}
        GreenCurve {keys [<at 0 0.715 RCIM_Cubic>,<at 1 0.79 RCIM_Cubic>]}
        RedCurve {keys [<at 0 0.62 RCIM_Cubic>,<at 1 0.7 RCIM_Cubic>]}
       }}))
    ScaleRGBA=true
   ) {module=ScaleColor001}
   &ScaleColor(
    Scale Alpha=&FloatFromCurve(
     FloatCurve=$System.ScaleDistanceFadeCurve
     CurveIndex=&Multiply_Float(
      A=&VectorLength(Vector=&Subtract_Vector(
        A=Linked() {linkedVariable {name=Particles.Position, type=NiagaraPosition}}
        B=&Add_Vector(A=Linked() {linkedVariable {name=Engine.Owner.Position, type=NiagaraPosition}}, B=<0 0 110>)
       ))
      B=$System.ScaleDistanceBias
     )
     Scale Curve=10
    )
    ScaleA=true
    ScaleRGB=true
   ) {module=ScaleColor002}
  ]
  rendererClasses [&NiagaraSpriteRendererProperties]
  renderers [{rendererClass=&NiagaraSpriteRendererProperties}]
 }
 Land_SnowPuffs_Torso {
  emitterSpawnScript [/SetVariables(Emitter.RandomSeed:NiagaraInt32=&RandomInt(Randomness Mode:ENiagaraRandomnessMode=::NewEnumerator0, Random Seed:NiagaraInt32=0, Fixed Random Seed:NiagaraBool=false)) {isSetParameters=true, module=SetVariables_#3}]
  emitterUpdateScript [
   &EmitterState(Life Cycle Mode=::NewEnumerator1, Loop Duration=1, UseLoopDelay=false)
   &SpawnRate(Spawn Probability=$User.LandBurst, SpawnRate=150, Use Spawn Probability=true)
  ]
  particleSpawnScript [
   &InitializeParticle(
    Color=&ColorFromCurve(
     ColorCurve=DataInterface() {propertyValues {
       AlphaCurve {keys [<at 0 0.06 RCIM_Cubic>,<at 1 0.02 RCIM_Cubic>]}
       BlueCurve {keys [<at 0 1 RCIM_Cubic>,<at 1 1 RCIM_Cubic>]}
       GreenCurve {keys [<at 0 1 RCIM_Cubic>,<at 1 0.8149 RCIM_Cubic>]}
       RedCurve {keys [<at 0 1 RCIM_Cubic>,<at 1 0.8 RCIM_Cubic>]}
      }}
     CurveIndex=$Particles.Partition.NormalizedID
    )
    Lifetime=&RandomRangeFloat(Minimum=0.7, Maximum=1.2)
    Module.Write Color=true
    Module.Write Lifetime=true
    Module.Write Mass=true
    Module.Write Material Random=true
    Module.Write Position=false
    Module.Write Scale=false
    Module.Write SpriteRotation=true
    Module.Write SpriteSize=true
    Position=&SimulationPosition()
    Sprite Rotation=&RandomRangeFloat(Minimum=-180, Maximum=180)
    Sprite Size=&Vector2DFromFloat(Value=&RandomRangeFloat(Minimum=8, Maximum=10))
   )
   &SkeletalMeshLocation(
    Bone Sampling Mode=::NewEnumerator0
    Inherit Velocity=false
    MaskSpawnGroup=false
    Module.Override Randomness=false
    Skeletal Mesh=$User.SkeletalMesh
   )
   &AddVelocity(Velocity=&Multiply_VectorByFloat(Vector=$Output.FN_BoneFlow.SplineTangent, Float=10))
   &AddVelocity(Velocity=<0 0 20>) {module=AddVelocity001}
   &InheritVelocity(Module.Limit Speed=true, Velocity Scale=<0.5 0.5 0.5>)
  ]
  particleUpdateScript [
   &AccelerationForce(Acceleration=&RandomRangeVector(Minimum=<0 0 -60>, Maximum=<0 0 -100>))
   &AddVelocityFromPoint(
    Module.Use Falloff Distance=false
    Origin Offset=<0 0 75>
    Velocity Strength=&FloatFromCurve(
     FloatCurve=DataInterface() {propertyValues {Curve {keys [
         {interpMode=RCIM_Cubic, tangentMode=RCTM_User, time=0, value=0}
         {arriveTangent=0.0249, interpMode=RCIM_Cubic, leaveTangent=-3.8646, tangentMode=RCTM_Break, time=0.3, value=1}
         <at 1 0>
        ]}}}
     Scale Curve=5
    )
   )
   &Drag(Drag=2, Use Linear Drag=true, Use Rotational Drag=false)
   &SolveForcesAndVelocity(Clamp Velocity=false, Limit Acceleration=false)
   &ParticleState()
   &SubUVAnimation(
    Sub UV Play Rate:NiagaraFloat=&Multiply_Float(A=0.5, B=$Particles.NormalizedAge)
    SubUV Animation Mode:ENiagara_SubUVLookupMode=::NewEnumerator3
   ) {module=SubUVAnimation001}
   &SpriteRotationRate(Rotation Rate=&FloatFromCurve(
     FloatCurve=DataInterface() {propertyValues {Curve {keys [<at 0 0 RCIM_Cubic>,<at 0.3 1 RCIM_Cubic>,<at 1 0>]}}}
     Scale Curve=&RandomRangeFloat(Minimum=-20, Maximum=20)
    ))
   &ScaleSpriteSize(Scale Factor=&Vector2DFromFloat(Value=&FloatFromCurve(
      FloatCurve=DataInterface() {propertyValues {Curve {keys [
          {arriveTangent=0.7564, interpMode=RCIM_Cubic, leaveTangent=0.7564, tangentMode=RCTM_User, time=-0.0024, value=0.2}
          {arriveTangent=1.4669, interpMode=RCIM_Cubic, leaveTangent=1.4669, time=0.3797, value=1.3341}
          <at 0.9951 1.6633 RCIM_Cubic>
         ]}}}
      Scale Curve=3
     )))
   &ScaleColor(
    Scale Mode=::NewEnumerator1
    Scale RGBA=&MakeVector4_FromLinearColor(LinearColor=&ColorFromCurve(ColorCurve=DataInterface() {propertyValues {
        AlphaCurve {keys [
          <at 0 0 RCIM_Cubic>
          {arriveTangent=0.6593, interpMode=RCIM_Cubic, leaveTangent=0.6593, time=0.2192, value=0.45}
          {arriveTangent=-1.2318, interpMode=RCIM_Cubic, leaveTangent=-1.2318, tangentMode=RCTM_User, time=0.5309, value=0.35}
          <at 1.0115 0>
          <at 1.094 0>
         ]}
        BlueCurve {keys [<at 0 1 RCIM_Cubic>,<at 1 1 RCIM_Cubic>]}
        GreenCurve {keys [<at 0 0.715 RCIM_Cubic>,<at 1 0.79 RCIM_Cubic>]}
        RedCurve {keys [<at 0 0.62 RCIM_Cubic>,<at 1 0.7 RCIM_Cubic>]}
       }}))
    ScaleRGBA=true
   ) {module=ScaleColor001}
   &ScaleColor(
    Scale Alpha=&FloatFromCurve(
     FloatCurve=$System.ScaleDistanceFadeCurve
     CurveIndex=&Multiply_Float(
      A=&VectorLength(Vector=&Subtract_Vector(
        A=Linked() {linkedVariable {name=Particles.Position, type=NiagaraPosition}}
        B=&Add_Vector(A=Linked() {linkedVariable {name=Engine.Owner.Position, type=NiagaraPosition}}, B=<0 0 60>)
       ))
      B=$System.ScaleDistanceBias
     )
     Scale Curve=5
    )
    ScaleA=true
    ScaleRGB=true
   ) {module=ScaleColor002}
  ]
  rendererClasses [&NiagaraSpriteRendererProperties]
  renderers [{rendererClass=&NiagaraSpriteRendererProperties}]
 }
 Landing_SnowFlakes {
  emitterUpdateScript [
   &EmitterState(Life Cycle Mode=::NewEnumerator1, Loop Duration=1, UseLoopDelay=false)
   &SpawnRate(Spawn Probability=$User.LandBurst, SpawnRate=250, Use Spawn Probability=true)
  ]
  particleSpawnScript [
   &InitializeParticle(
    Lifetime=&RandomRangeFloat(Minimum=0.7, Maximum=1.2)
    Module.Write Color=true
    Module.Write Lifetime=true
    Module.Write Mass=true
    Module.Write Material Random=true
    Module.Write Position=false
    Module.Write Scale=false
    Module.Write SpriteRotation=true
    Module.Write SpriteSize=true
    Position=&SimulationPosition()
    Sprite Rotation=&RandomRangeFloat(Minimum=0, Maximum=360)
    Sprite Size=&Vector2DFromFloat(Value=&RandomRangeFloat(Minimum=3, Maximum=5))
   )
   &Color(Color=<rgba 0.4 0.59 1 1>, Scale Alpha=$Particles.MaterialRandom)
   &SkeletalMeshLocation(
    Bone Sampling Mode=::NewEnumerator0
    Inherit Velocity=false
    MaskSpawnGroup=false
    Module.Override Randomness=false
    Skeletal Mesh=$User.SkeletalMesh
   )
   &AddVelocityFromPoint(
    Module.Use Falloff Distance=false
    Origin Offset=<0 0 100>
    Velocity Strength=&RandomRangeFloat(Minimum=50, Maximum=75)
   )
   &AddVelocity(Velocity=<0 0 -10>)
   /SetVariables(Particles.RandFacing:Vector3f=&RandomRangeVector(Minimum=<-1 -1 -1>, Maximum=<1 1 1>)) {isSetParameters=true, module=SetVariables_#5}
   &InheritVelocity(Module.Limit Speed=true, Velocity Scale=<0.3 0.3 0.3>)
  ]
  particleUpdateScript [
   &ParticleState()
   &CurlNoiseForce(Noise Frequency=100, Noise Strength=200, Pan Noise=false)
   &AccelerationForce(Acceleration=<0 0 -50>)
   &Drag(Drag=&RandomRangeFloat(Minimum=2, Maximum=3), Use Linear Drag=true, Use Rotational Drag=false)
   &SolveForcesAndVelocity(Clamp Velocity=false, Limit Acceleration=false)
   &ScaleColor(
    Scale RGB=&MakeVectorFromLinearColorRGB(Color=&ColorFromCurve(ColorCurve=DataInterface() {propertyValues {
        AlphaCurve {keys [<at 0 0 RCIM_Cubic>,<at 0.04 1>,<at 0.135 0.62>,<at 1.0115 0>,<at 1.094 0>]}
        BlueCurve {keys [<at 0 1 RCIM_Cubic>,<at 1 1 RCIM_Cubic>]}
        GreenCurve {keys [<at 0 0.8875 RCIM_Cubic>,<at 1 0.79 RCIM_Cubic>]}
        RedCurve {keys [<at 0 0.85 RCIM_Cubic>,<at 1 0.7 RCIM_Cubic>]}
       }}))
    ScaleA=true
    ScaleRGB=true
   )
   &ScaleSpriteSize(Scale Factor=&Vector2DFromFloat(Value=&FloatFromCurve(FloatCurve=DataInterface() {propertyValues {Curve {keys [
          <at 0 0 RCIM_Cubic>
          {arriveTangent=2, interpMode=RCIM_Cubic, leaveTangent=2, time=0.1619, value=1}
          {arriveTangent=-1.1932, interpMode=RCIM_Cubic, leaveTangent=-1.1932, time=0.5, value=1}
          <at 1 0 RCIM_Cubic>
         ]}}})))
   &FN_SpriteAlignmentAndFacing(SpriteFacing:Vector3f=&Lerp_Vector(
     A=&RandomRangeVector(Minimum=<-1 -1 -1>, Maximum=<1 1 1>)
     B=$Particles.RandFacing
     Alpha=$Particles.NormalizedAge
    ))
  ]
  rendererClasses [&NiagaraSpriteRendererProperties]
  renderers [{rendererClass=&NiagaraSpriteRendererProperties}]
 }
}
summary {
 systemName=NS_WeepingWoods_Winter_Idle_Jump
 emitters [
  {emitterName=Landing_SnowFlakes, rendererClasses [&NiagaraSpriteRendererProperties]}
  {emitterName=Land_SnowPuffs_Torso, rendererClasses [&NiagaraSpriteRendererProperties]}
  {emitterName=Land_Leaves, rendererClasses [&NiagaraMeshRendererProperties]}
  {emitterName=Idle_Leaves, rendererClasses [&NiagaraMeshRendererProperties]}
  {emitterName=Idle_SnowFlakes, rendererClasses [&NiagaraSpriteRendererProperties]}
  {emitterName=Land_SnowPuffs_Head, rendererClasses [&NiagaraSpriteRendererProperties]}
 ]
 userVariables [
  User.bIsFrontEnd {type=NiagaraBool}
  User.bIsFrontEndPreview {type=NiagaraBool}
  User.BlackSmokeTrailSpawnRate {type=NiagaraFloat, defaultValue=10, description=""}
  User.ColorChange {type=LinearColor, description="", defaultValue {struct=LinearColor, value=<rgba 1 1 1 1>}}
  User.EyeGlowMultiplier {type=NiagaraFloat, defaultValue=0, description=""}
  User.EyeGlowStrength {type=NiagaraFloat, defaultValue=0, description=""}
  User.JumpBurst {type=NiagaraFloat, defaultValue=0, description=""}
  User.JumpTimeRange {type=NiagaraFloat, defaultValue=0, description=""}
  User.LandBurst {type=NiagaraFloat, defaultValue=0, description=""}
  User.LightStrength {type=NiagaraFloat, defaultValue=0, description=""}
  User.Right2ImpactVelocity {type=Vector3f, description="", defaultValue {struct=Vector3f, value=<0 0 0>}}
  User.SkeletalMesh {
   type=NiagaraDataInterfaceSkeletalMesh
   defaultValue {
    struct=NiagaraExt_VariableValue_DataInterface
    value {dataInterface=&self:NiagaraDataInterfaceSkeletalMesh_1}
   }
  }
 ]
}
